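# Stereo MetaSound source for the GnatGala music encounter's idle state, mixing "Perc" and "Else" stems with separate gains.
# One gain combines master volume, a fade-in envelope, smoothed distance attenuation and a pause fade, then scales both channels.
# Attenuation is computed in-graph from distance, inner radius and falloff inputs.
# Pausing starts a delay matching the fade-out that fires OnFinished, and unpausing resets it.
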MetaSoundSource /FNM_AI_GnatGala/Audio/GnatGalaMusic/MSS_GnatGala_Music_Idle
inputs [
 Idle Else {type=WaveAsset}
 Idle Perc {type=WaveAsset}
 MasterVolume {type=Float}
 Mix.Else {type=Float}
 Mix.Perc {type=Float}
 MusicEncounter.Attenuation.Distance {type=Float}
 MusicEncounter.Attenuation.Volume.FalloffDistance {type=Float}
 MusicEncounter.Attenuation.Volume.InnerRadius {type=Float}
 MusicEncounter.bIsPaused {type=Bool}
 MusicEncounter.PauseFadeOutTime {type=Time}
 UE.Source.OnPlay {type=Trigger}
]
interfaces [UE.OutputFormat.Stereo:1.0, UE.Source.OneShot:1.0, UE.Source:1.1]
outputs [
 "UE.OutputFormat.Stereo.Audio:0" {type=Audio}
 "UE.OutputFormat.Stereo.Audio:1" {type=Audio}
 UE.Source.OneShot.OnFinished {type=Trigger}
]
pages [Default {
  input_defaults {
   Idle Else=/Game/Sounds/Creative/Gadgets/Radio/Music_Survival_Horror/Wavs/MX_SurvivalHorror_HallOfMirrors_Loop_01
   Idle Perc=null
   MasterVolume=1
   Mix.Else=1
   Mix.Perc=1
   MusicEncounter.Attenuation.Distance=0
   MusicEncounter.Attenuation.Volume.FalloffDistance=10000
   MusicEncounter.Attenuation.Volume.InnerRadius=6000
   MusicEncounter.PauseFadeOutTime=3
   MusicEncounter.bIsPaused=null
   UE.Source.OnPlay=null
  }
  nodes [
   InitVariable.Trigger()
   WavePlayer_0 = UE.Wave Player:Stereo(Wave Asset=^Idle Perc, Play=^UE.Source.OnPlay) {also {defaults {Loop="<MetaSoundBuilderResult.SUCCEEDED: 0>"}}}
   WavePlayer_1 = UE.Wave Player:Stereo(Wave Asset=^Idle Else, Play=^UE.Source.OnPlay) {also {defaults {Loop="<MetaSoundBuilderResult.SUCCEEDED: 0>"}}}
   AudioMixerAudioMixerStereo2_0 = "AudioMixer.Audio Mixer (Stereo, 2)"(
    Gain 0=^Mix.Else
    Gain 1=^Mix.Perc
    In 1 L=@WavePlayer_0.Out Left
    In 1 R=@WavePlayer_0.Out Right
    In 0 L=@WavePlayer_1.Out Left
    In 0 R=@WavePlayer_1.Out Right
   )
   n1 = UE.Add:Float(AdditionalOperands=^MusicEncounter.Attenuation.Volume.FalloffDistance, PrimaryOperand=^MusicEncounter.Attenuation.Volume.InnerRadius)
   n2 = UE.InterpTo:Audio(Target=^MusicEncounter.Attenuation.Distance) {also {defaults {Interp Time=0.5}}}
   ADSREnvelopeADSREnvelope_0 = ADSR Envelope.ADSR Envelope:Float(Trigger Attack=^UE.Source.OnPlay) {also {defaults {Attack Time=2, Decay Time=0, Sustain Level=1}}}
   TriggerOnValueChange_0 = UE.Trigger On Value Change:Bool(Value=^MusicEncounter.bIsPaused)
   n3 = TriggerCompare.Clamp:Bool(A=^MusicEncounter.bIsPaused, Compare=@TriggerOnValueChange_0.Trigger) {also {defaults {B="<MetaSoundBuilderResult.SUCCEEDED: 0>"}}}
   n4 = Value.Value:Float(Reset=@n3.False, Set=@n3.True) {also {defaults {Init Value=1, Target Value=0}}}
   n5 = UE.InterpTo:Audio(Interp Time=^MusicEncounter.PauseFadeOutTime, Target=@n4.Output Value)
   n6 = UE.Multiply:Float(
    AdditionalOperands=^MasterVolume
    PrimaryOperand=UE.Multiply:Float(
     PrimaryOperand=MapRange.MapRange:Float(In Range A=^MusicEncounter.Attenuation.Volume.InnerRadius, In Range B=@n1.Out, In=@n2.Value) {also {defaults {Out Range A=1, Out Range B=0}}}.Out Value
     AdditionalOperands=UE.Multiply:Float(PrimaryOperand=@ADSREnvelopeADSREnvelope_0.Out Envelope, AdditionalOperands=@n5.Value).Out
    ).Out
   )
   Multiply_0 = UE.Multiply:Audio by Float(PrimaryOperand=@AudioMixerAudioMixerStereo2_0.Out R, AdditionalOperands=@n6.Out)
   Multiply_1 = UE.Multiply:Audio by Float(PrimaryOperand=@AudioMixerAudioMixerStereo2_0.Out L, AdditionalOperands=@n6.Out)
   TriggerDelay_0 = UE.Trigger Delay(Delay Time=^MusicEncounter.PauseFadeOutTime, Reset=@n4.On Reset, In=@n4.On Set)
  ]
  links [
   {dst="@UE.OutputFormat.Stereo.Audio:1", src=Multiply_0.Out}
   {dst="@UE.OutputFormat.Stereo.Audio:0", src=Multiply_1.Out}
   {dst="@UE.Source.OneShot.OnFinished", src=TriggerDelay_0.Out}
  ]
 }]
properties {
 assetImportData=&self:AssetImportData
 numChannels=2
 outputFormat=Stereo
 sampleRate=48000
 soundClassObject=&Fort_Music_NPC
 assetUserData [&self:AudioAssetUserData_0]
}
