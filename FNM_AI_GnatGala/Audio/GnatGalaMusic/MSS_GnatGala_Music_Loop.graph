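# Stereo MetaSound music for the Gnat Gala boss encounter, switching from a loop to an ending piece by boss phase.
# A Harmonix metronome quantizes phase-driven starts of both wave players, with an ADSR envelope releasing the loop as the ending begins.
# Gain combines a pause fade, smoothed distance through a wavetable curve and a fixed boost.
# The sound finishes when the ending completes or the pause fade time elapses.
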MetaSoundSource /FNM_AI_GnatGala/Audio/GnatGalaMusic/MSS_GnatGala_Music_Loop
inputs [
 MusicEncounter.Attenuation.Distance {type=Float}
 MusicEncounter.Attenuation.Volume.FalloffDistance {type=Float}
 MusicEncounter.Attenuation.Volume.InnerRadius {type=Float}
 MusicEncounter.bIsPaused {type=Bool}
 MusicEncounter.Boss.Phase {type=Int32}
 MusicEncounter.PauseFadeOutTime {type=Time}
 Tempo {type=Float}
 Tempo 1 {type=Float, access=Constructor}
 Tempo 2 {type=Float}
 UE.Attenuation.Distance {type=Float}
 UE.Source.OnPlay {type=Trigger}
]
interfaces [UE.Attenuation:1.0, UE.OutputFormat.Stereo:1.0, UE.Source.OneShot:1.0, UE.Source:1.1]
outputs [
 "UE.OutputFormat.Stereo.Audio:0" {type=Audio}
 "UE.OutputFormat.Stereo.Audio:1" {type=Audio}
 UE.Source.OneShot.OnFinished {type=Trigger}
]
pages [Default {
  input_defaults {
   MusicEncounter.Attenuation.Distance=0
   MusicEncounter.Attenuation.Volume.FalloffDistance=10000
   MusicEncounter.Attenuation.Volume.InnerRadius=6000
   MusicEncounter.Boss.Phase=0
   MusicEncounter.PauseFadeOutTime=3
   MusicEncounter.bIsPaused=null
   Tempo=160
   Tempo 1=140
   Tempo 2=160
   UE.Attenuation.Distance=0
   UE.Source.OnPlay=null
  }
  nodes [
   n1 = HarmonixNodes.Metronome(Tempo=^Tempo, Transport=HarmonixNodes.TriggerToTransport(Play=^UE.Source.OnPlay).Transport) {also {defaults {Simple Loop="<MetaSoundBuilderResult.SUCCEEDED: 0>"}}}
   TriggerOnValueChange_1 = UE.Trigger On Value Change:Int32(Value=^MusicEncounter.Boss.Phase)
   TriggerAnyTriggerAny2_0 = "TriggerAny.Trigger Any (2)"(In 0=^UE.Source.OnPlay, In 1=@TriggerOnValueChange_1.Trigger)
   TriggerOnce_1 = UE.Trigger Once(Trigger In=HarmonixNodes.MidiClockSubdivisionTrigger(MIDI Clock=@n1.MIDI Clock).Trigger Out, Reset=@TriggerAnyTriggerAny2_0.Out)
   n2 = UE.Reroute(Value=@TriggerOnce_1.Trigger Out)
   TriggerDelay_1 = UE.Trigger Delay(Delay Time=UE.BPMToSeconds(BPM=^Tempo).Seconds, In=@n2.Value)
   TriggerOnce_0 = UE.Trigger Once(Trigger In=@n2.Value, Reset=@TriggerDelay_1.Out)
   n3 = HarmonixNodes.TransportWavePlayerController(
    Transport=HarmonixNodes.TriggerToTransport(Play=TriggerCompare.Clamp:Int32(A=^MusicEncounter.Boss.Phase, Compare=@TriggerOnce_0.Trigger Out) {also {defaults {B=1}}}.True).Transport
    MIDI Clock=UE.Reroute(Value=UE.Reroute(Value=@n1.MIDI Clock).Value).Value
   ) {also {version [0,1]}}
   WavePlayer_1 = UE.Wave Player:Stereo(Play=@n3.Play, Start Time=@n3.Start Time, Stop=@n3.Stop) {also {defaults {Wave Asset=/FNM_AI_GnatGala/Audio/GnatGala_Music_End}, version [1,1]}}
   ADSREnvelopeADSREnvelope_0 = ADSR Envelope.ADSR Envelope:Float(
    Trigger Attack=^UE.Source.OnPlay
    Release Time=UE.BPMToSeconds() {also {defaults {BPM=160, Beat Multiplier=2, Divisions of Whole Note=1}}}.Seconds
    Attack Time=UE.BPMToSeconds() {also {defaults {BPM=160}}}.Seconds
    Trigger Release=@WavePlayer_1.On Play
   ) {also {defaults {Sustain Level=1}}}
   DecibelstoLinearGain_0 = UE.Decibels to Linear Gain:Float() {also {defaults {Decibels=12}}}
   TriggerOnValueChange_0 = UE.Trigger On Value Change:Bool(Value=^MusicEncounter.bIsPaused)
   n4 = TriggerCompare.Clamp:Bool(A=^MusicEncounter.bIsPaused, Compare=@TriggerOnValueChange_0.Trigger) {also {defaults {B="<MetaSoundBuilderResult.SUCCEEDED: 0>"}}}
   n5 = Value.Value:Float(Reset=@n4.False, Set=@n4.True) {also {defaults {Init Value=1, Target Value=0}}}
   n6 = UE.InterpTo:Audio(Interp Time=^MusicEncounter.PauseFadeOutTime, Target=@n5.Output Value)
   n7 = MapRange.MapRange:Float(In=UE.InterpTo:Audio(Target=^UE.Attenuation.Distance) {also {defaults {Interp Time=0.5}}}.Value) {also {defaults {In Range B=2750}}}
   n8 = UE.Multiply:Float(
    AdditionalOperands=@DecibelstoLinearGain_0.Linear Gain
    PrimaryOperand=UE.Multiply:Float(
     PrimaryOperand=UE.Reroute(Value=@n6.Value).Value
     AdditionalOperands=UE.WaveTableBankEvaluate(Input=@n7.Out Value) {also {defaults {WaveTableBank=&/WTB_GnatGala_ME}}}.Output
    ).Out
   )
   n9 = UE.Reroute(Value=@TriggerOnce_1.Trigger Out)
   n10 = HarmonixNodes.TransportWavePlayerController(
    Transport=HarmonixNodes.TriggerToTransport(Play=TriggerCompare.Clamp:Int32(A=^MusicEncounter.Boss.Phase, Compare=@n9.Value).True).Transport
    MIDI Clock=UE.Reroute(Value=UE.Reroute(Value=@n1.MIDI Clock).Value).Value
   ) {also {version [0,1]}}
   WavePlayer_0 = UE.Wave Player:Stereo(Play=@n10.Play, Start Time=@n10.Start Time, Stop=@n10.Stop) {also {defaults {Loop="<MetaSoundBuilderResult.SUCCEEDED: 0>", Wave Asset=/FNM_AI_GnatGala/Audio/GnatGala_Music_Loop}, version [1,1]}}
   AudioMixerAudioMixerStereo2_0 = "AudioMixer.Audio Mixer (Stereo, 2)"(
    Gain 0=UE.Multiply:Float(PrimaryOperand=@ADSREnvelopeADSREnvelope_0.Out Envelope, AdditionalOperands=@n8.Out).Out
    Gain 1=@n8.Out
    In 0 L=@WavePlayer_0.Out Left
    In 0 R=@WavePlayer_0.Out Right
    In 1 L=@WavePlayer_1.Out Left
    In 1 R=@WavePlayer_1.Out Right
   )
   TriggerDelay_0 = UE.Trigger Delay(Delay Time=^MusicEncounter.PauseFadeOutTime, Reset=@n5.On Reset, In=@n5.On Set)
   TriggerAnyTriggerAny2_1 = "TriggerAny.Trigger Any (2)"(In 1=@TriggerDelay_0.Out, In 0=@WavePlayer_1.On Finished)
  ]
  links [
   {dst="@UE.OutputFormat.Stereo.Audio:0", src=AudioMixerAudioMixerStereo2_0.Out L}
   {dst="@UE.OutputFormat.Stereo.Audio:1", src=AudioMixerAudioMixerStereo2_0.Out R}
   {dst="@UE.Source.OneShot.OnFinished", src=TriggerAnyTriggerAny2_1.Out}
  ]
 }]
properties {
 assetImportData=&self:AssetImportData
 numChannels=2
 outputFormat=Stereo
 sampleRate=48000
 soundClassObject=&Fort_Music_NPC
 assetUserData [&self:AudioAssetUserData_0]
 concurrencyOverrides {maxCount=1}
}
